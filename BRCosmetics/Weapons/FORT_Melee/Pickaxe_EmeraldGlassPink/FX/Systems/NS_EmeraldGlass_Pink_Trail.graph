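# Niagara weapon trail for the pink Emerald Glass pickaxe, drawing three short-lived ribbon trails from the User.SkeletalMesh.
# Glow and SoftTrail span bone/socket indices 1 and 0, while SoftTrail001 spans 3 and 2 with the shortest lifetime.
# Curves fade alpha over life, and Glow and SoftTrail also darken from pink or white toward deep purple.
# A time-of-day multiplier dims both SoftTrail emitters at night.
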NiagaraSystem /BRCosmetics/Weapons/FORT_Melee/Pickaxe_EmeraldGlassPink/FX/Systems/NS_EmeraldGlass_Pink_Trail
dependencies {
 usedDataInterfaces [&NiagaraDataInterfaceColorCurve, &NiagaraDataInterfaceCurve]
 usedDynamicInputs []
 usedModules [&SystemState]
}
emitters {
 Glow {
  emitterUpdateScript [
   &EmitterState()
   &FN_SpawnAnimTrail()
  ]
  particleSpawnScript [
   &FN_InitializeAnimTrailBoneIndecies(First Bone or Socket Index=1, Second Bone or Socket Index=0)
   &FN_InitializeAnimTrail(AnimTrailWidth=0.75, Lifetime=0.175)
  ]
  particleUpdateScript [
   &FN_UpdateAnimTrail(SkeletalMesh=$User.SkeletalMesh) {module=FN_UpdateAnimTrail001}
   &SolveForcesAndVelocity(Clamp Velocity=false, Limit Acceleration=false)
   &ScaleColor(
    Scale Alpha=&FloatFromCurve(
     FloatCurve=DataInterface() {propertyValues {Curve {keys [
         {tangentMode=RCTM_User, time=0, value=1}
         {arriveTangent=-0.0405, interpMode=RCIM_Cubic, leaveTangent=-0.0405, tangentMode=RCTM_User, time=0.7009, value=1}
         {arriveTangent=-8.0089, interpMode=RCIM_Cubic, leaveTangent=-8.0089, tangentMode=RCTM_User, time=1, value=0}
        ]}}}
     Scale Curve=0.3
    )
    Scale RGB=&Multiply_Vector(
     A=&MakeVectorFromLinearColorRGB(Color=&ColorFromCurve(ColorCurve=DataInterface() {propertyValues {
         AlphaCurve {keys [<at 0 1 RCIM_Cubic>]}
         BlueCurve {keys [<at 0 0.8606>,<at 0.6701 0.1228 RCIM_Cubic>]}
         GreenCurve {keys [<at 0 0.4635>,<at 0.6701 0.0187 RCIM_Cubic>]}
         RedCurve {keys [<at 0 1>,<at 0.6701 0.1875 RCIM_Cubic>]}
        }}))
     B=&FN_Select_Vector_from_Time_Of_Day(
      Morning=&MakeVectorFromLinearColorRGB(Color=<rgba 1 1 1 1>)
      Day=&MakeVectorFromLinearColorRGB(Color=<rgba 1 1 1 1>)
      Evening=&MakeVectorFromLinearColorRGB(Color=<rgba 1 1 1 1>)
      Night=&MakeVectorFromLinearColorRGB(Color=<rgba 1 1 1 1>)
     )
    )
    ScaleA=true
    ScaleRGB=true
   ) {module=ScaleColor001}
  ]
  rendererClasses [&NiagaraRibbonRendererProperties]
  renderers [{rendererClass=&NiagaraRibbonRendererProperties}]
 }
 SoftTrail {
  emitterUpdateScript [
   &EmitterState()
   &FN_SpawnAnimTrail()
  ]
  particleSpawnScript [
   &FN_InitializeAnimTrailBoneIndecies(First Bone or Socket Index=1, Second Bone or Socket Index=0)
   &FN_InitializeAnimTrail(Lifetime=0.2)
  ]
  particleUpdateScript [
   &FN_UpdateAnimTrail(SkeletalMesh=$User.SkeletalMesh) {module=FN_UpdateAnimTrail001}
   &SolveForcesAndVelocity(Clamp Velocity=false, Limit Acceleration=false)
   &ScaleColor(
    Scale Alpha=&FloatFromCurve(
     FloatCurve=DataInterface() {propertyValues {Curve {keys [
         {arriveTangent=-1.8862, interpMode=RCIM_Cubic, leaveTangent=-1.8862, tangentMode=RCTM_User, time=0, value=1.25}
         {interpMode=RCIM_Cubic, tangentMode=RCTM_User, time=1, value=0}
        ]}}}
     Scale Curve=0.5
    )
    Scale RGB=&Multiply_Vector(
     A=&MakeVectorFromLinearColorRGB(Color=&ColorFromCurve(ColorCurve=DataInterface() {propertyValues {
         AlphaCurve {keys [<at 0 1 RCIM_Cubic>]}
         BlueCurve {keys [<at 0 1 RCIM_Cubic>,<at 0.5957 0.1464 RCIM_Cubic>]}
         GreenCurve {keys [<at 0 1 RCIM_Cubic>,<at 0.5957 0 RCIM_Cubic>]}
         RedCurve {keys [<at 0 1 RCIM_Cubic>,<at 0.5957 0.3125 RCIM_Cubic>]}
        }}))
     B=&FN_Select_Vector_from_Time_Of_Day(
      Morning=&MakeVectorFromLinearColorRGB(Color=<rgba 1 1 1 1>)
      Day=&MakeVectorFromLinearColorRGB(Color=<rgba 1 1 1 1>)
      Evening=&MakeVectorFromLinearColorRGB(Color=<rgba 1 1 1 1>)
      Night=&MakeVectorFromLinearColorRGB(Color=<rgba 0.33 0.33 0.33 1>)
     )
    )
    ScaleA=true
    ScaleRGB=true
   ) {module=ScaleColor001}
  ]
  rendererClasses [&NiagaraRibbonRendererProperties]
  renderers [{rendererClass=&NiagaraRibbonRendererProperties}]
 }
 SoftTrail001 {
  emitterUpdateScript [
   &EmitterState()
   &FN_SpawnAnimTrail()
  ]
  particleSpawnScript [
   &FN_InitializeAnimTrailBoneIndecies(First Bone or Socket Index=3, Second Bone or Socket Index=2)
   &FN_InitializeAnimTrail(Lifetime=0.1)
  ]
  particleUpdateScript [
   &FN_UpdateAnimTrail(SkeletalMesh=$User.SkeletalMesh) {module=FN_UpdateAnimTrail001}
   &SolveForcesAndVelocity(Clamp Velocity=false, Limit Acceleration=false)
   &ScaleColor(
    Scale Alpha=&FloatFromCurve(FloatCurve=DataInterface() {propertyValues {Curve {keys [
         {arriveTangent=-1.8862, interpMode=RCIM_Cubic, leaveTangent=-1.8862, tangentMode=RCTM_User, time=0, value=1.25}
         {interpMode=RCIM_Cubic, tangentMode=RCTM_User, time=1, value=0}
        ]}}})
    Scale RGB=&FN_Select_Vector_from_Time_Of_Day(
     Morning=&MakeVectorFromLinearColorRGB(Color=<rgba 1 1 1 1>)
     Day=&MakeVectorFromLinearColorRGB(Color=<rgba 1 1 1 1>)
     Evening=&MakeVectorFromLinearColorRGB(Color=<rgba 1 1 1 1>)
     Night=&MakeVectorFromLinearColorRGB(Color=<rgba 0.5 0.5 0.5 1>)
    )
    ScaleA=true
    ScaleRGB=true
   ) {module=ScaleColor001}
  ]
  rendererClasses [&NiagaraRibbonRendererProperties]
  renderers [{rendererClass=&NiagaraRibbonRendererProperties}]
 }
}
summary {
 systemName=NS_EmeraldGlass_Pink_Trail
 emitters [
  {emitterName=SoftTrail, rendererClasses [&NiagaraRibbonRendererProperties]}
  {emitterName=Glow, rendererClasses [&NiagaraRibbonRendererProperties]}
  {emitterName=SoftTrail001, rendererClasses [&NiagaraRibbonRendererProperties]}
 ]
 userVariables [
  User.RibbonIDOffset {type=NiagaraInt32}
  User.SkeletalMesh {
   type=NiagaraDataInterfaceSkeletalMesh
   defaultValue {
    struct=NiagaraExt_VariableValue_DataInterface
    value {dataInterface=&self:NiagaraDataInterfaceSkeletalMesh_0}
   }
  }
  User.VariantPicker {type=NiagaraFloat, defaultValue=0, description=""}
 ]
}
