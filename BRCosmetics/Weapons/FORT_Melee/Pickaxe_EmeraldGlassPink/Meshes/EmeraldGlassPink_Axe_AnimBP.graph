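# Animation Blueprint for the EmeraldGlassPink pickaxe prop, which snaps closed during melee animations and opens again afterwards.
# Each update reads the MeleeAnim curve from the owning player's FortPlayerAnimInstance and sets Melee Anim Playing when it is at least 1.
# The state machine enters at Into_Idle, advances automatically to looping Idle, blends briefly to the Closed pose while that flag is set, then returns through Into_Idle.
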AnimBlueprint /BRCosmetics/Weapons/FORT_Melee/Pickaxe_EmeraldGlassPink/Meshes/EmeraldGlassPink_Axe_AnimBP
target_skeleton=&/EmeraldGlassPink_Axe_Skeleton
graphs [
 AnimGraph {nodes [
   AnimGraphNode_StateMachine_0 = EmeraldGlassPink State Machine() {is=title AnimGraphNode_StateMachine}
   Output Pose(Result=@AnimGraphNode_StateMachine_0) {layerGroup=None}
   AnimGraphNode_SubInput_0 = Input Pose()
  ]}
 EmeraldGlassPink State Machine {
  entry=Into_Idle
  path=AnimGraph.AnimGraphNode_StateMachine_0.EmeraldGlassPink State Machine
  ids=id
  nodes [
   AnimStateNode_0 = Closed() {is=title AnimStateNode}
   AnimStateNode_1 = Into_Idle() {is=title AnimStateNode}
   AnimStateNode_2 = Idle()
   AnimStateTransitionNode_0 = AnimStateTransitionNode() {bAutomaticRuleBasedOnSequencePlayerInState=true, crossfadeDuration=0}
   AnimStateTransitionNode_1 = AnimStateTransitionNode() {crossfadeDuration=0.1}
   AnimStateTransitionNode_2 = AnimStateTransitionNode() {crossfadeDuration=0}
  ]
  links [
   {
    dst=Into_Idle
    rule=AnimGraph.AnimGraphNode_StateMachine_0.EmeraldGlassPink State Machine.AnimStateTransitionNode_2.Transition
    src=Closed
    via=AnimStateTransitionNode_2
   }
   {
    dst=Closed
    rule=AnimGraph.AnimGraphNode_StateMachine_0.EmeraldGlassPink State Machine.AnimStateTransitionNode_1.Transition
    src=Idle
    via=AnimStateTransitionNode_1
   }
   {
    dst=Idle
    rule=AnimGraph.AnimGraphNode_StateMachine_0.EmeraldGlassPink State Machine.AnimStateTransitionNode_0.Transition
    src=Into_Idle
    via=AnimStateTransitionNode_0
   }
  ]
 }
 Closed {
  path=AnimGraph.AnimGraphNode_StateMachine_0.EmeraldGlassPink State Machine.AnimStateNode_0.Closed
  nodes [Output Animation Pose(Result=Sequence Player 'EmeraldGlassPink_ClosedPose_PROP'() {bLoopAnimation=false, sequence=/BRCosmetics/Weapons/FORT_Melee/Pickaxe_EmeraldGlassPink/Animation/EmeraldGlassPink_ClosedPose_PROP})]
 }
 Into_Idle {
  path=AnimGraph.AnimGraphNode_StateMachine_0.EmeraldGlassPink State Machine.AnimStateNode_1.Into_Idle
  nodes [Output Animation Pose(Result=Sequence Player 'EmeraldGlassPink_ToOpen_PROP'() {bLoopAnimation=false, sequence=/BRCosmetics/Weapons/FORT_Melee/Pickaxe_EmeraldGlassPink/Animation/EmeraldGlassPink_ToOpen_PROP})]
 }
 Idle {
  path=AnimGraph.AnimGraphNode_StateMachine_0.EmeraldGlassPink State Machine.AnimStateNode_2.Idle
  nodes [Output Animation Pose(Result=Sequence Player 'EmeraldGlassPink_Idle_PROP'() {sequence=/BRCosmetics/Weapons/FORT_Melee/Pickaxe_EmeraldGlassPink/Animation/EmeraldGlassPink_Idle_PROP})]
 }
 Transition {path=AnimGraph.AnimGraphNode_StateMachine_0.EmeraldGlassPink State Machine.AnimStateTransitionNode_0.Transition, nodes [Result()]}
 Transition {
  path=AnimGraph.AnimGraphNode_StateMachine_0.EmeraldGlassPink State Machine.AnimStateTransitionNode_1.Transition
  nodes [Result(bCanEnterTransition=Get Melee Anim Playing() -> (Melee Anim Playing:bool))]
 }
 Transition {
  path=AnimGraph.AnimGraphNode_StateMachine_0.EmeraldGlassPink State Machine.AnimStateTransitionNode_2.Transition
  nodes [Result(bCanEnterTransition=NOT Boolean(A=Get Melee Anim Playing()))]
 }
 EventGraph {nodes [
   Event BlueprintUpdateAnimation()
   > n1 = Cast To FortWeaponPickaxeAthena!(execute, Object=GetOwningActor()) -> (then, CastFailed, AsFort Weapon Pickaxe Athena:Fort Weapon Pickaxe Athena Object Reference) {is=title K2Node_DynamicCast}
   n2 = GetOwner(self:Actor Object Reference=@n1)
   > Is Valid(InputObject=@n2)
   n3 = GetComponentByClass!(self:Actor Object Reference=@n2, ComponentClass:Actor Component Class Reference=/Script/Engine.SkeletalMeshComponent) -> (ReturnValue:Skeletal Mesh Component Object Reference)
   > Is Valid(InputObject=@n3)
   > n4 = Cast To FortPlayerAnimInstance(Object=GetAnimInstance(self=@n3))
   > Set Melee Anim Playing(Melee Anim Playing:bool="float >= float"(A=GetCurveValue(self=@n4, CurveName=MeleeAnim), B=1) {is=title K2Node_CallFunction}) -> (Output_Get:bool)
   TryGetPawnOwner()
  ]}
]
variables [Melee Anim Playing {type=boolean}]
